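# Niagara system for the LetsPlay player emote that makes confetti pops with glowing flares above the effect origin.
# LocationGenerator spawns short-lived particles at random overhead points and sends location events, apparently seeding Confetti_Pop and Pop_Flare, though no receiving handlers appear.
# Confetti sprites fall with randomized drag and gravity, collide, shrink late in life and animate through a randomly offset dynamic material parameter.
# Pop_Flare adds a brief, bright red sprite with a light renderer.
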NiagaraSystem /BRCosmetics/Animation/Game/MainPlayer/Emotes/LetsPlay/FX/NS_LetsPlay_Confetti
dependencies {
 usedDataInterfaces [&NiagaraDataInterfaceCurve, &NiagaraDataInterfaceVectorCurve]
 usedDynamicInputs []
 usedModules [&SystemState]
}
emitters {
 Confetti_Pop {
  emitterUpdateScript [&EmitterState()]
  particleSpawnScript [
   /Niagara/Modules/Spawn/Initialization/InitializeParticle(
    Color=&MakeLinearColorFromVectorAndFloat(
     "Vector (RGB)"=&MakeVector(X=1, Y=0, Z=&UniformAOrBFloat(A=0.2, B=0.6))
     "Float (Alpha)"=1
    )
    Lifetime=&UniformRangedFloat(Minimum=1.5, Maximum=1.7)
    Module.Write Color=true
    Module.Write Lifetime=true
    Module.Write Mass=true
    Module.Write Material Random=true
    Module.Write Position=false
    Module.Write Scale=false
    Module.Write SpriteRotation=true
    Module.Write SpriteSize=true
    Position=&SimulationPosition()
    Sprite Rotation=&UniformRangedFloat(Minimum=0, Maximum=180)
    Sprite Size=&Vector2DFromFloat(Value=&UniformRangedFloat(Minimum=4, Maximum=5))
   )
   /SetVariables(Particles.Stable_Gravity:Vector3f=&UniformRangedVector(Minimum=<0 0 -150>, Maximum=<1 1 -370>)) {isSetParameters=true}
   /SetVariables(Particles.Stable_Drag:NiagaraFloat=&UniformRangedFloat(Minimum=2, Maximum=2.2)) {isSetParameters=true}
   /SetVariables(Particles.RandomFlip:NiagaraFloat=&UniformRangedFloat(Minimum=60, Maximum=90)) {isSetParameters=true}
  ]
  particleUpdateScript [
   &ParticleState()
   &Drag(Drag=$Particles.Initial.Stable_Drag, Use Linear Drag=true, Use Rotational Drag=false)
   &GravityForce(Gravity=$Particles.Initial.Stable_Gravity)
   &Collision(
    Advanced Aging Rate=1
    Age Colliding Particles=false
    Analytical Collision Plane Position 1:Vector3f=<0 0 0>
    Analytical Collision Plane Space:ENiagaraCoordinateSpace=::Local
    CPU Collision Type=::NewEnumerator3
    Module.Control Rest State Via Surface Normal=false
    Module.EnableMaxCollisionCount=false
    Module.Randomize Collision Normal Vector=false
    Particle Radius Scale=0.2
    Perform 2 Analytical Collision Queries=false
    Restitution=0.15
   )
   &SolveForcesAndVelocity(Clamp Velocity=false, Limit Acceleration=false)
   &ScaleSpriteSize(Scale Factor=&Vector2DFromFloat(Value=&FloatFromCurve(FloatCurve=DataInterface() {propertyValues {Curve {keys [<at 0.85 1>, {arriveTangent=-2.5206, interpMode=RCIM_Cubic, leaveTangent=-2.5206, tangentMode=RCTM_User, time=1, value=0}]}}})))
   &ScaleColor(
    Scale RGB=&VectorFromCurve(VectorCurve=DataInterface() {propertyValues {
       XCurve {keys [<at 0 3 RCIM_Cubic>, {arriveTangent=-2, interpMode=RCIM_Cubic, leaveTangent=-2, time=0.2442, value=1}, <at 1 1>]}
       YCurve {keys [{arriveTangent=436.3616, time=0, value=0}, {leaveTangent=29.5933, time=1, value=0}]}
       ZCurve {keys [<at 0 1 RCIM_Cubic>]}
      }})
    ScaleA=false
    ScaleRGB=false
   )
   &DynamicMaterialParameters(
    Index 0 Param 1=&Add_Float(
     A=&Multiply_Float(A=$Particles.Initial.MaterialRandom, B=36)
     B=&FloatFromCurve(FloatCurve=DataInterface() {propertyValues {Curve {keys [<at 0 0>,<at 0.6 108>]}}})
    )
    Module.Param0WriteEnabled=true
    Module.Param1WriteEnabled=false
    Module.Param2WriteEnabled=false
    Module.Param3WriteEnabled=false
   ) {module=DynamicMaterialParameters001}
  ]
  rendererClasses [&NiagaraSpriteRendererProperties]
  renderers [{rendererClass=&NiagaraSpriteRendererProperties}]
 }
 LocationGenerator {
  emitterUpdateScript [
   &EmitterState()
   &SpawnRate(Spawn Probability=0.6, SpawnRate=10, Use Spawn Probability=true)
  ]
  particleSpawnScript [
   &InitializeParticle(
    Lifetime=0.1
    Module.Write Color=true
    Module.Write Lifetime=true
    Module.Write Mass=true
    Module.Write Material Random=true
    Module.Write Position=false
    Module.Write Scale=false
    Module.Write SpriteRotation=false
    Module.Write SpriteSize=true
    Position=&SimulationPosition()
    Sprite Size=<20 20>
   )
   /Niagara/Modules/Spawn/Location/SphereLocation(
    MaskSpawnGroup=false
    Module.Override Randomness=false
    Offset=&UniformRangedVector(Minimum=<0 0 220>, Maximum=<1 1 300>)
    Sphere Radius=&UniformRangedFloat(Minimum=10, Maximum=50)
    Surface Only=false
   )
   /SetVariables(Particles.SpawnEvent=true) {isSetParameters=true}
  ]
  particleUpdateScript [
   &ParticleState()
   &SolveForcesAndVelocity(Clamp Velocity=false, Limit Acceleration=false)
   &ScaleSpriteSize(Scale Factor=&Vector2DFromFloat(Value=&FloatFromCurve(FloatCurve=DataInterface() {propertyValues {Curve {keys [<at 0 1 RCIM_Cubic>,<at 1 0>]}}})))
   &GenerateLocationEvent(Event Send Rate=600, EventCondition:NiagaraBool=$Particles.SpawnEvent, Module.Use Event Send Rate=true)
   /SetVariables() {isSetParameters=true}
  ]
  rendererClasses [&NiagaraSpriteRendererProperties]
  renderers [{rendererClass=&NiagaraSpriteRendererProperties}]
 }
 Pop_Flare {
  emitterUpdateScript [&EmitterState()]
  particleSpawnScript [&InitializeParticle(
    Color=<rgba 1 0 0 0.06>
    Lifetime=&UniformRangedFloat(Minimum=0.05, Maximum=0.15)
    Module.Write Color=true
    Module.Write Lifetime=true
    Module.Write Mass=true
    Module.Write Material Random=true
    Module.Write Position=false
    Module.Write Scale=false
    Module.Write SpriteRotation=true
    Module.Write SpriteSize=true
    Position=&SimulationPosition()
    Sprite Rotation=&UniformRangedFloat(Minimum=0, Maximum=1)
    Sprite Size=&Vector2DFromFloat(Value=&UniformRangedFloat(Minimum=190, Maximum=260))
   )]
  particleUpdateScript [
   &ParticleState()
   &SolveForcesAndVelocity(Clamp Velocity=false, Limit Acceleration=false)
   &ScaleSpriteSize(Scale Factor=&Vector2DFromFloat(Value=&FloatFromCurve(FloatCurve=DataInterface() {propertyValues {Curve {keys [<at 0 0 RCIM_Cubic>,<at 0.1 1 RCIM_Cubic>,<at 1 0>]}}})))
   &ScaleColor(
    Scale Alpha=&FloatFromCurve(FloatCurve=DataInterface() {propertyValues {Curve {keys [
         {arriveTangent=3.9711, interpMode=RCIM_Cubic, leaveTangent=3.9711, tangentMode=RCTM_User, time=0, value=0}
         <at 0.5 1 RCIM_Cubic>
         {arriveTangent=-3.9883, interpMode=RCIM_Cubic, leaveTangent=-3.9883, tangentMode=RCTM_User, time=1, value=0}
        ]}}})
    Scale RGB=&VectorFromFloat(Value=20)
    ScaleA=true
    ScaleRGB=true
   )
   /SetVariables(Particles.LightColor=&LerpLinearColors(StartColor=$Particles.Color, EndColor=<rgba 0 0 0 0>, LerpFactor=0.95)) {isSetParameters=true}
  ]
  rendererClasses [&NiagaraSpriteRendererProperties, &NiagaraLightRendererProperties]
  renderers [
   {rendererClass=&NiagaraSpriteRendererProperties}
   {rendererClass=&NiagaraLightRendererProperties, rendererIndex=1}
  ]
 }
}
summary {
 systemName=NS_LetsPlay_Confetti
 emitters [
  {emitterName=LocationGenerator, rendererClasses [&NiagaraSpriteRendererProperties]}
  {emitterName=Confetti_Pop, rendererClasses [&NiagaraSpriteRendererProperties]}
  {emitterName=Pop_Flare, rendererClasses [&NiagaraSpriteRendererProperties, &NiagaraLightRendererProperties]}
 ]
 userVariables [User.bIsFrontEndPreview {type=NiagaraBool, defaultValue=true}]
}
